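# Physics asset for the Wayfare gadget prop mesh in the BRCosmetics animation props.
# It holds a single body on JNT_Wayfare_Main, shaped as one rotated capsule offset from the joint.
# There are no constraints, so the gadget's collision is represented as one piece with no jointed parts.
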PhysicsAsset /BRCosmetics/Animation/Props/Gadgets_Wayfare/Meshes/Wayfare_Gadget_Physics
bodies {JNT_Wayfare_Main [{shapeType=Capsule, length=11.3056, radius=11.2074, center=<0.0033 -10.8674 3.6611>, rotation=<rot -0.5535 -90 -86.6407>}]}
constraints []
